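# Conversation database for the Voltaic Huntress NPC, Saeko, providing her greeting and top-level conversation branches.
# The entry point leads to one speech node whose contextual messages cover a first-ever meeting, a Palespooky outfit reaction, and spirit and Underworld lines gated by match conversation count.
# Priority curve rows rank the first-meeting and outfit lines.
# The speech node branches through link nodes to separate entry tags for NPC library quests, basic services and her own dialogue.
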ConversationDatabase /NPCLibrary/NPCs/VoltaicHuntress/NPCLibrary_Conversation_VoltaicHuntress
entry_points [{tag=Conversation.Entry.VoltaicHuntress, destinations [#1]}]
nodes [
 ConversationEntryPointNode_7 {
  id=#1
  class=ConversationEntryPointNode
  outputs [#2]
  properties {entryTag {tagName=Conversation.Entry.VoltaicHuntress}}
 }
 Speech_C_4 {
  id=#2
  class=Speech_C
  outputs [#3, #4, #5]
  properties {
   nodeName=""
   generalConfig {
    defaultMessage=Shhh. Can you hear the spirits murmuring?
    contextualMessages [
     {
      message=My name is Saeko. I hunt the dead.
      contextRequirements [{requirement {class=FortControllerRequirement_IsFirstEverConversationWithNPC}}]
      priority {curve {rowName=Tandem.Speech.Priority.FirstEverInteraction}}
     }
     {
      message="Be calm, restless spirit. Be at rest."
      contextRequirements [{requirement {
         class=FortControllerRequirement_HasCID
         properties {allowedCIds [/BRCosmetics/Athena/Items/Cosmetics/Characters/CID_601_Athena_Commando_F_Palespooky]}
        }}]
      priority {curve {rowName=Tandem.Speech.Priority.OutfitReaction}}
     }
     {
      message=Shhh. Can you hear the spirits murmuring?
      contextRequirements [{requirement {
         class=FortControllerRequirement_CheckNPCMatchConversationCount
         properties {nPCUniqueIdTag {tagName=AISpawnerData.Type.Tandem.VoltaicHuntress}}
        }}]
      priority {curve {rowName=None}}
     }
     {
      message=The Underworld is causing chaos for the Island's spirits.
      contextRequirements [{requirement {
         class=FortControllerRequirement_CheckNPCMatchConversationCount
         properties {nPCUniqueIdTag {tagName=AISpawnerData.Type.Tandem.VoltaicHuntress}}
        }}]
      priority {curve {rowName=None}}
     }
     {
      message=I heard the boatman Charon retired. Another headache for me.
      contextRequirements [{requirement {
         class=FortControllerRequirement_CheckNPCMatchConversationCount
         properties {nPCUniqueIdTag {tagName=AISpawnerData.Type.Tandem.VoltaicHuntress}}
        }}]
      priority {curve {rowName=None}}
     }
     {
      message=The spirits are restless. I have about a dozen theories as to why.
      contextRequirements [{requirement {
         class=FortControllerRequirement_CheckNPCMatchConversationCount
         properties {nPCUniqueIdTag {tagName=AISpawnerData.Type.Tandem.VoltaicHuntress}}
        }}]
      priority {curve {rowName=None}}
     }
     {
      message="Hades needs to spend less time in battle, and more time running his realm."
      contextRequirements [{requirement {
         class=FortControllerRequirement_CheckNPCMatchConversationCount
         properties {nPCUniqueIdTag {tagName=AISpawnerData.Type.Tandem.VoltaicHuntress}}
        }}]
      priority {curve {rowName=None}}
     }
    ]
   }
  }
 }
 ConversationLinkNode_2 {
  id=#5
  class=ConversationLinkNode
  properties {remoteEntryTag {tagName=Conversation.Entry.VoltaicHuntress.Dialogue}}
 }
 ConversationLinkNode_0 {id=#4, class=ConversationLinkNode, properties {remoteEntryTag {tagName=Conversation.Entry.BasicServices}}}
 ConversationLinkNode_1 {id=#3, class=ConversationLinkNode, properties {remoteEntryTag {tagName=Conversation.Entry.NPCLibrary.Quests}}}
]
