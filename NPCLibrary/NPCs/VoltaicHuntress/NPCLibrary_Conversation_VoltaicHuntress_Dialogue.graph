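# Conversation database for the Voltaic Huntress NPC's optional "Ask about..." lore dialogue, entered through its Dialogue entry tag.
# An opening line about the spirit realm branches into choices on The Underworld, Midas and Hunters, plus a Back option.
# Each topic plays a linear chain of three speech nodes, then converges on one shared ReturnToStart node labeled BACK.
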ConversationDatabase /NPCLibrary/NPCs/VoltaicHuntress/NPCLibrary_Conversation_VoltaicHuntress_Dialogue
entry_points [{tag=Conversation.Entry.VoltaicHuntress.Dialogue, destinations [#1]}]
nodes [
 ConversationEntryPointNode_0 {
  id=#1
  class=ConversationEntryPointNode
  outputs [#2]
  properties {entryTag {tagName=Conversation.Entry.VoltaicHuntress.Dialogue}}
 }
 Speech_C_0 {
  id=#2
  class=Speech_C
  outputs [#3, #4, #5, #6]
  properties {
   nodeName=""
   softTaskIcon=&T_UI_NPC_question
   generalConfig {defaultMessage=You are curious about the realm of the spirits?}
  }
  sub_nodes [ConversationChoiceNode_5 {class=ConversationChoiceNode, properties {defaultChoiceDisplayText=Ask about...}}]
 }
 Speech_C_2 {
  id=#6
  class=Speech_C
  outputs [#7]
  properties {
   nodeName=""
   softTaskIcon=&T_UI_NPC_question
   generalConfig {defaultMessage=There are many of us who hunt the strange and dangerous.}
  }
  sub_nodes [ConversationChoiceNode_10 {class=ConversationChoiceNode, properties {defaultChoiceDisplayText=Hunters}}]
 }
 Speech_C_3 {
  id=#7
  class=Speech_C
  outputs [#8]
  properties {
   nodeName=""
   softTaskIcon=&T_UI_NPC_question
   generalConfig {defaultMessage=We keep in touch with one another. Share intel. Offer support.}
  }
 }
 Speech_C_4 {
  id=#8
  class=Speech_C
  outputs [#9]
  properties {
   nodeName=""
   softTaskIcon=&T_UI_NPC_question
   generalConfig {defaultMessage="Take care, however. Monsters can be... wily."}
  }
 }
 ReturnToStart_C_23 {
  id=#9
  class=ReturnToStart_C
  properties {softTaskIcon=&T_UI_NPC_Back}
  sub_nodes [ConversationChoiceNode_32 {class=ConversationChoiceNode, properties {defaultChoiceDisplayText=BACK}}]
 }
 Back_C_0 {
  id=#5
  class=Back_C
  sub_nodes [ConversationChoiceNode_7 {class=ConversationChoiceNode, properties {defaultChoiceDisplayText=BACK}}]
 }
 Speech_C_6 {
  id=#4
  class=Speech_C
  outputs [#10]
  properties {
   nodeName=""
   softTaskIcon=&T_UI_NPC_question
   generalConfig {defaultMessage="It is rare, that a spirit can escape The Underworld."}
  }
  sub_nodes [ConversationChoiceNode_0 {class=ConversationChoiceNode, properties {defaultChoiceDisplayText=Midas}}]
 }
 Speech_C_12 {
  id=#10
  class=Speech_C
  outputs [#11]
  properties {
   nodeName=""
   softTaskIcon=&T_UI_NPC_question
   generalConfig {defaultMessage=This Midas will return to a half life... but a life nonetheless.}
  }
 }
 Speech_C_13 {
  id=#11
  class=Speech_C
  outputs [#9]
  properties {nodeName="", softTaskIcon=&T_UI_NPC_question, generalConfig {defaultMessage=May he find what he seeks.}}
 }
 Speech_C_7 {
  id=#3
  class=Speech_C
  outputs [#12]
  properties {nodeName="", softTaskIcon=&T_UI_NPC_question, generalConfig {defaultMessage=The Island's spirits are restless.}}
  sub_nodes [ConversationChoiceNode_15 {class=ConversationChoiceNode, properties {defaultChoiceDisplayText=The Underworld}}]
 }
 Speech_C_15 {
  id=#12
  class=Speech_C
  outputs [#13]
  properties {
   nodeName=""
   softTaskIcon=&T_UI_NPC_question
   generalConfig {defaultMessage=Is it the lack of a Stygian boatman? The distraction of its king?}
  }
 }
 Speech_C_14 {
  id=#13
  class=Speech_C
  outputs [#9]
  properties {
   nodeName=""
   softTaskIcon=&T_UI_NPC_question
   generalConfig {defaultMessage=Or is that cursed Pandora's Box aggravating the spirits?}
  }
 }
]
